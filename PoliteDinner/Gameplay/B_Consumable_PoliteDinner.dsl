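; Blueprint for the PoliteDinner consumable weapon, a noodle item whose mesh plays a material-driven progress animation.
; On equip, it stops any still-playing synced consume montage on the weapon mesh.
; PlayProgress_TA passes game time, stage and loop duration to the mesh materials.
; That duration lerps between minimum and maximum tunables by the instigator's missing shield relative to max healing percent, falling to minimum without a player pawn.
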(event FortWeapon|EventOnEquippedAndAttached
  (bind _weaponmesh (Variables|Visual|GetWeaponMesh))
  (bind _returnvalue (Components|SkeletalMesh|GetAnimInstance _weaponmesh))
  (bind _syncedconsumemontage (Variables|Default|GetSyncedConsumeMontage))
  ;[0.03429 0.3069 0.7899] Ensure we stop drinking montage when reequipped
  (if (Animation|Montage|MontageIsPlaying _returnvalue _syncedconsumemontage)
    (Animation|Montage|MontageStop _returnvalue 0.0 _syncedconsumemontage)))

(event Custom|PlayProgress_TA
  (bind _weaponmesh (Variables|Visual|GetWeaponMesh))
  ; PlayProgress_TA
  (bind _duration (CallFunction|GetNoodleLoopDuration))
  (Rendering|Material|SetScalarParameterValueonMaterials _weaponmesh "AnimationDuration" _duration) ; Set Duration.
  (Rendering|Material|SetScalarParameterValueonMaterials _weaponmesh "Stage" 1.0) ; Stage1
  (Rendering|Material|SetScalarParameterValueonMaterials _weaponmesh "GameTimeFromBP" (Utilities|Time|GetGameTimeinSeconds)) ; Passthough Time
)

(fn GetNoodleLoopDuration
  (bind _asfort_player_pawn (Utilities|Casting|CastToFortPlayerPawn (Game|GetInstigator))
    (:then
      (LocalVariables|SetAlpha (Math|Float|Min(Float) (/ (- 1.0 (FortPawn|ShieldPercent _asfort_player_pawn)) (ScalableFloat|GetCachedScalableFloatValue (Variables|NoodleLevel|GetMaxHealingPercent))) 1.0)))
    (:CastFailed
      (LocalVariables|SetAlpha 0.0)))
  (return (Math|Float|Lerp (ScalableFloat|GetCachedScalableFloatValue (Variables|NoodleLevel|GetMinNoodleDuration)) (ScalableFloat|GetCachedScalableFloatValue (Variables|NoodleLevel|GetMaxHealingDuration)) (LocalVariables|GetAlpha))))

(fn ConstructionScript
  (|CallParentFunction))
