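; Thrown projectile Blueprint for the PoliteDinner item that customizes where its throw sound plays.
; PlayThrowSound calls the parent when the throw sound is player-attached; otherwise, where cosmetic events can run, it uses the owner weapon's sound library component.
; It temporarily points that component's playback settings at the projectile root, plays the Throw event, then resets the settings.
; The event graph and construction script add no logic of their own.
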(event EventBeginPlay
  (|CallParentFunction) ; This node is disabled and will not be called.\nDrag off pins to build functionality.
)

(event Collision|EventActorBeginOverlap
  (params (OtherActor :type Actor%20Object%20Reference)))

(event EventTick
  (params (DeltaSeconds :type Float%20%28single-precision%29)))

(fn PlayThrowSound
  ; Attach throw sound library sound to projectile instead of player if there is no player attached sound for throwing shields
  (Variables|Sound|GetThrowSoundAttachedToPlayer
    (:then
      (|CallParentFunction))
    (:else
      (Networking|CanExecuteCosmeticEvents
        (:True
          (bind _returnvalue (CallFunction|GetOwnerWeapon))
          (Utilities|IsValid _returnvalue
            (:Is%20Valid
              (Utilities|IsValid (SoundLibrary|GetSoundLibraryComponent _returnvalue "(TagName=\"\")" :!class /Script/FortniteGame.FortWeapon)
                (:Is%20Valid
                  (Class|FortSoundLibraryComponent|SetPlaybackSettings (SoundLibrary|GetSoundLibraryComponent _returnvalue "(TagName=\"\")" :!class /Script/FortniteGame.FortWeapon) (Utilities|Struct|MakeFortSoundLibraryPlaybackSettings true false true true 1.0 1.0 0.0 0 0 "None" (Variables|Transformation|GetRootComponent)))
                  (SoundLibrary|PlayEvent (SoundLibrary|GetSoundLibraryComponent _returnvalue "(TagName=\"\")" :!class /Script/FortniteGame.FortWeapon) "(TagName=\"SoundLibrary.Item.Throw\")")
                  (Class|FortSoundLibraryComponent|SetPlaybackSettings (SoundLibrary|GetSoundLibraryComponent _returnvalue "(TagName=\"\")" :!class /Script/FortniteGame.FortWeapon) (Utilities|Struct|MakeFortSoundLibraryPlaybackSettings false)))))))))))

(fn ConstructionScript
  (|CallParentFunction))
