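; PoliteDinner heal ability that restores health and shield through its parent, then grants buffs and stamina on end.
; EventHeal heals only while magnitudes allow it and the pawn has the channeling tag.
; When nothing is left to heal, a float-gated server-wait net sync precedes EndAbility, working around missing VFX on tiny heals.
; EndAbility sends a gameplay event, applies the buff effect and adds energy to the TacticalSprintEnergyComponent, without checking cancellation.
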(event EventHeal
  ;[1 1 1] Ends the ability if cannot heal of Shields and Health are full after healing. This will apply the buffs
  (if (CallFunction|CheckCanHeal)
    (|CallParentFunction)
    (bind (_healmagnitude _shieldmagnitude _applyeffect) (CallFunction|GetMagnitudes))
    (if _applyeffect
      _
      (elif (> (ScalableFloat|GetCachedScalableFloatValue (Variables|Default|GetSF_WaitForClientToEndAbility)) 0)
        ;;[1 0.2417 0.3825] Workaround to fix an issue of missing VFXs if the amount healed was too little
        (Ability|Tasks|WaitNetSync "OnlyServerWait"
          (:OnSync
            (Ability|EndAbility)))
        (else
          _)))
    (else
      _)))

(event Ability|EventOnEndAbility
  (params (bWasCancelled :type Boolean))
  (bind _player_pawn (CallFunction|GetCachedPawn))
  (bind _returnvalue (Actor|GetComponentbyClass _player_pawn "/TacticalSprintGame/Gameplay/TacticalSprintEnergyComponent.TacticalSprintEnergyComponent_C"))
  ;[1 1 1] Apply Buffs and adds Energy when ability ends
  (|CallParentFunction bWasCancelled)
  (Ability|SendGameplayEventtoActor (CallFunction|GetCachedPawn) (Variables|Default|GetEvent%20Tag))
  (Ability|ApplyGameplayEffectSpecToOwner (Ability|MakeOutgoingGameplayEffectSpec (Variables|Default|GetGE_Buffs)))
  (Utilities|IsValid _returnvalue
    (:Is%20Valid
      (Energy|AddEnergy _returnvalue (ScalableFloat|GetCachedScalableFloatValue (Variables|Default|GetOneShotStaminaRegen)) 0 "(GameplayTags=)"))))

(fn CheckCanHeal
  ;[1 1 1] Only Heals if HP or Shield isn't full AND is still playing the consume montage
  (bind (_healmagnitude _shieldmagnitude _applyeffect) (CallFunction|GetMagnitudes))
  (return (and _applyeffect (GameplayTags|HasAnyMatchingGameplayTags (CallFunction|GetCachedPawn) (Variables|Default|GetChannelingTag)))))

(fn GetCachedPawn
  (Utilities|IsValid (Variables|Default|GetPlayer%20Pawn)
    (:Is%20Valid
      _)
    (:Is%20Not%20Valid
      (bind _outfortplayerpawn (CallFunction|GetAvatarasValidFortPlayerPawn)
        (:ValidFortPlayerPawn
          (Variables|Default|SetPlayer%20Pawn _outfortplayerpawn)))))
  (return (Variables|Default|GetPlayer%20Pawn)))
