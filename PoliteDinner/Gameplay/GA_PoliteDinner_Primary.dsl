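; Primary gameplay ability for PoliteDinner, extending a parent incremental-healing ability with custom montage handling.
; The incremental healing override only applies the SetupGE effect to the owner, so full health does not interrupt the animation.
; If FullyHealedTag is valid, activation starts a gameplay-event wait that jumps the montage to its outro; ending the ability ends that task.
; A Chooser selects the montage info, with gender and body type set from the player pawn.
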(event Ability|EventActivateAbility
  ;[1 1 1] When activating, also start Event to wait for full heal Event from the Effective Heal GA
  (|CallParentFunction)
  (if (GameplayTags|IsGameplayTagValid (Variables|Default|GetFullyHealedTag))
    (CallFunction|WaitFullHeal)))

(event EventDoIncrementalHealing
  ;[1 1 1] Override DoIncrementalHealing because we don't actually need the healing part and don't want animation to get interrupted when health is full
  (bind _effect_out (CallFunction|SetupGE))
  (Ability|ApplyGameplayEffectSpecToOwner _effect_out))

(event Custom|WaitFullHeal
  ;[1 1 1] Wait for full heal to trigger Montage's Outro
  (bind _asynctaskproxy (Ability|Tasks|WaitGameplayEvent (Variables|Default|GetFullyHealedTag) 0 true)
    (:then
      (Variables|Default|SetWaitFullHealTask _asynctaskproxy))
    (:EventReceived
      (Ability|Animation|MontageJumptoSection (Variables|Default|GetMontageOutroSectionName)))))

(event Ability|EventOnEndAbility
  (params (bWasCancelled :type Boolean))
  (|CallParentFunction bWasCancelled)
  (GameplayTasks|EndTask (Variables|Default|GetWaitFullHealTask)))

(fn MakeCustomMontageInfo
  (bind _playerpawn (Variables|Default|GetPlayerPawn))
  (bind (_ismale _isfemale) (FortPlayerPawn|GetGender _playerpawn))
  (|CallParentFunction)
  (Variables|Default|SetIsFemale _isfemale)
  (bind _returnvalue (Animation|ConvertCustomBodyTypetoAnimBodyType (Customization|GetCharacterBodyType _playerpawn)))
  (Variables|Default|SetAnimBodyType _returnvalue)
  (bind _result (Animation|EvaluateChooser self)
    (:execute
      (bind _structout (Utilities|Struct|SetmembersinFortGameplayAbilityMontageInfo (Variables|Animation|GetMontageInfo) _result))
      (Variables|Animation|SetMontageInfo _structout))))
